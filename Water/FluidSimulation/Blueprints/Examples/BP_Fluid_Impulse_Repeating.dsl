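; BP_Fluid_Impulse_Repeating is an example Blueprint actor that repeatedly applies a force impulse to a fluid simulation.
; On BeginPlay it caches the first BP_FluidSim_01 actor found in the level.
; A fixed time step on Tick, set by an interval variable, sends the stored impulse settings, updated with the actor's current location, to the sim's ApplyFluidForceImpulse.
; The overlap event and construction script are empty.
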(event EventBeginPlay
  (bind _outactors (Actor|GetAllActorsOfClass "/FortWater/FluidSimulation/Blueprints/BP_FluidSim_01.BP_FluidSim_01_C"))
  (Variables|Default|SetSim (Utilities|Array|Get(acopy) _outactors)))

(event Collision|EventActorBeginOverlap
  (params (OtherActor :type Actor%20Object%20Reference)))

(event EventTick
  (params (DeltaSeconds :type Float%20%28single-precision%29))
  (Utilities|FixedTimeStep (Variables|Default|GetImpulse%20Every%20N%20Seconds)
    (:Fixed%20Step%20Execution
      (bind _structout (Utilities|Struct|SetmembersinFluidForceImpulse (Variables|Default|GetImulse%20Settings) (Transformation|GetActorLocation)))
      (Class|BPFluidSim01|ApplyFluidForceImpulse (Variables|Default|GetSim) _structout))))

(fn ConstructionScript)
